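# Animation Blueprint for the penguin backpack cosmetic that adds secondary physics motion to its wings, comb, scarf and flippers.
# The input pose passes through six chained Anim Dynamics nodes in component space, one per dyn_backpack bone, then converts back to local space.
# Each bone has tight angular limits; comb, scarf and flippers also use angular springs, with overridden damping on the scarf.
# The event graph holds only an unconnected update event.
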AnimBlueprint /BRCosmetics/Accessories/FORT_Backpacks/Backpack_Penguin/Meshes/Backpack_Penguin_AnimBP
target_skeleton=&/Backpack_Penguin_Skeleton
graphs [
 AnimGraph {nodes [
   AnimGraphNode_SubInput_0 = Input Pose()
   n1 = Anim Dynamics - Bone: dyn_backpack_wing_l(ComponentPose=Local To Component(LocalPose=@AnimGraphNode_SubInput_0), Alpha=1, ExternalForce="0,0,0") {
    alpha=0.9
    angularBiasOverride=0.99
    angularSpringConstant=1000
    bGravityOverrideInSimSpace=true
    numSolverIterationsPostUpdate=4
    numSolverIterationsPreUpdate=16
    alphaBoolBlend {blendOption=Linear}
    boundBone {boneName=dyn_backpack_wing_l}
    componentLinearAccScale=<1 1 1>
    componentLinearVelScale=<1 1 1>
    physicsBodyDefinitions [{
      sphereCollisionRadius=0
      boundBone {boneName=dyn_backpack_wing_l}
      boxExtents=<10 5 10>
      constraintSetup {angularTargetAxis=AxisZ, twistAxis=AxisZ, angularLimitsMax=<0 20 0>, angularLimitsMin=<0 -10 0>, angularTarget=<0 0 1>}
      localJointOffset=<1 0 0>
     }]
   }
   n2 = Anim Dynamics - Bone: dyn_backpack_wing_r(ComponentPose=@n1, Alpha=1, ExternalForce="0,0,0") {
    alpha=0.9
    angularBiasOverride=0.99
    angularSpringConstant=1000
    bGravityOverrideInSimSpace=true
    numSolverIterationsPostUpdate=4
    numSolverIterationsPreUpdate=16
    alphaBoolBlend {blendOption=Linear}
    boundBone {boneName=dyn_backpack_wing_r}
    componentLinearAccScale=<1 1 1>
    componentLinearVelScale=<1 1 1>
    physicsBodyDefinitions [{
      sphereCollisionRadius=0
      boundBone {boneName=dyn_backpack_wing_r}
      boxExtents=<10 5 10>
      constraintSetup {angularTargetAxis=AxisZ, twistAxis=AxisZ, angularLimitsMax=<0 20 0>, angularLimitsMin=<0 -10 0>, angularTarget=<0 0 1>}
      localJointOffset=<1 0 0>
     }]
   }
   n3 = Anim Dynamics - Bone: dyn_backpack_comb(ComponentPose=@n2, Alpha=1, ExternalForce="0,0,0") {
    alpha=0.9
    angularSpringConstant=600
    bAngularSpring=true
    bGravityOverrideInSimSpace=true
    numSolverIterationsPostUpdate=4
    numSolverIterationsPreUpdate=16
    alphaBoolBlend {blendOption=Linear}
    boundBone {boneName=dyn_backpack_comb}
    componentLinearAccScale=<1 1 1>
    componentLinearVelScale=<1 1 1>
    physicsBodyDefinitions [{
      sphereCollisionRadius=0
      boundBone {boneName=dyn_backpack_comb}
      boxExtents=<3 2 2>
      constraintSetup {angularTargetAxis=AxisZ, angularLimitsMax=<10 10 10>, angularLimitsMin=<-10 -10 -10>, angularTarget=<0 0 1>}
      localJointOffset=<1 0 0>
     }]
   }
   n4 = Anim Dynamics - Bone: dyn_backpack_scarf(ComponentPose=@n3, Alpha=1, ExternalForce="0,0,0") {
    alpha=0.9
    angularDampingOverride=0.9
    angularSpringConstant=500
    bAngularSpring=true
    bGravityOverrideInSimSpace=true
    bOverrideAngularDamping=true
    numSolverIterationsPostUpdate=4
    numSolverIterationsPreUpdate=16
    alphaBoolBlend {blendOption=Linear}
    boundBone {boneName=dyn_backpack_scarf}
    componentLinearAccScale=<1 1 1>
    componentLinearVelScale=<1 1 1>
    physicsBodyDefinitions [{
      sphereCollisionRadius=0
      boundBone {boneName=dyn_backpack_scarf}
      boxExtents=<5 2 5>
      constraintSetup {angularTargetAxis=AxisZ, angularLimitsMax=<5 10 20>, angularLimitsMin=<-5 -10 -20>, angularTarget=<0 0 1>}
      localJointOffset=<1 0 0>
     }]
   }
   n5 = Anim Dynamics - Bone: dyn_backpack_flipper_l(ComponentPose=@n4, Alpha=1, ExternalForce="0,0,0") {
    alpha=0.9
    angularSpringConstant=1000
    bAngularSpring=true
    bGravityOverrideInSimSpace=true
    numSolverIterationsPostUpdate=4
    numSolverIterationsPreUpdate=16
    alphaBoolBlend {blendOption=Linear}
    boundBone {boneName=dyn_backpack_flipper_l}
    componentLinearAccScale=<1 1 1>
    componentLinearVelScale=<1 1 1>
    physicsBodyDefinitions [{
      sphereCollisionRadius=0
      boundBone {boneName=dyn_backpack_flipper_l}
      boxExtents=<7 5 2>
      constraintSetup {angularTargetAxis=AxisY, twistAxis=AxisZ, angularLimitsMax=<5 10 5>, angularLimitsMin=<-5 -10 -5>, angularTarget=<0 1 0>}
      localJointOffset=<3 0 0>
     }]
   }
   n6 = Anim Dynamics - Bone: dyn_backpack_flipper_r(ComponentPose=@n5, Alpha=1, ExternalForce="0,0,0") {
    alpha=0.9
    angularSpringConstant=1000
    bAngularSpring=true
    bGravityOverrideInSimSpace=true
    numSolverIterationsPostUpdate=4
    numSolverIterationsPreUpdate=16
    alphaBoolBlend {blendOption=Linear}
    boundBone {boneName=dyn_backpack_flipper_r}
    componentLinearAccScale=<1 1 1>
    componentLinearVelScale=<1 1 1>
    physicsBodyDefinitions [{
      sphereCollisionRadius=0
      boundBone {boneName=dyn_backpack_flipper_r}
      boxExtents=<7 5 2>
      constraintSetup {angularTargetAxis=AxisY, twistAxis=AxisZ, angularLimitsMax=<5 10 5>, angularLimitsMin=<-5 -10 -5>, angularTarget=<0 1 0>}
      localJointOffset=<3 0 0>
     }]
   }
   Output Pose(Result=Component To Local(ComponentPose=@n6)) {layerGroup=None}
  ]}
 EventGraph {nodes [Event BlueprintUpdateAnimation()]}
]
variables []
