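# Physics asset for the Penguin backpack cosmetic, giving its skeletal mesh a simple physics shape.
# A single capsule body on the backpack_root bone, offset and rotated from the bone, approximates the backpack.
# It has no constraints, so it defines no jointed or secondary-motion chain.
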PhysicsAsset /BRCosmetics/Accessories/FORT_Backpacks/Backpack_Penguin/Meshes/Backpack_Penguin_PhysicsAsset
bodies {backpack_root [{shapeType=Capsule, length=16.6489, radius=11.5741, center=<2.0699 -9.6216 0.3878>, rotation=<rot 0 93.4406 -87.9779>}]}
constraints []
